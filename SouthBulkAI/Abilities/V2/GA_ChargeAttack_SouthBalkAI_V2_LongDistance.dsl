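; Long-distance charge attack ability for the SouthBalk V2 AI that adds a big beam to the dash of its parent ability.
; On dash it adds a looping cue and spawns the beam above the pawn, attached, with damage duration from RunTimer, ending the ability if spawning fails.
; Ability end applies the PlayerBreather effect to the owner.
; The custom source transform is raised by the capsule half-height.
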(event Ability|EventActivateAbility
  (|CallParentFunction) ; This node is disabled and will not be called.\nDrag off pins to build functionality.
)

(event Ability|EventOnEndAbility
  (params (bWasCancelled :type Boolean))
  (|CallParentFunction bWasCancelled)
  (Ability|ApplyGameplayEffectToOwner (Variables|Default|GetGE_PlayerBreather)))

(event EventDash
  (Ability|AddGameplayCueToOwner (Variables|Default|GetGC_DashLoop))
  (CallFunction|SpawnBeam)
  (|CallParentFunction))

(event Custom|SpawnBeam
  (bind (_returnvalue_location _returnvalue_rotation) (Math|Transform|BreakTransform (Transformation|GetActorTransform (Variables|Default|GetPlayerPawn))))
  (bind _returnvalue (Game|SpawnActorfromClass "/SouthBulkAI/Gameplay/BGA_SouthBalkAI_V2_BigBeam.BGA_SouthBalkAI_V2_BigBeam_C" (Math|Transform|MakeTransform (+ _returnvalue_location "0.000000,0.000000,100.000000") _returnvalue_rotation) "AlwaysSpawn" "MultiplyWithRoot" 0 0 (Variables|Default|GetPlayerPawn)))
  (Utilities|IsValid _returnvalue
    (:Is%20Valid
      (Transformation|AttachActorToActor _returnvalue (Variables|Default|GetPlayerPawn) "None" "KeepWorld" "KeepWorld" "KeepWorld" true)
      (Class|BGASouthBalkAIV2BigBeam|SetBeamDamageDuration _returnvalue (ScalableFloat|GetValueatLevel (Variables|Default|GetRunTimer))))
    (:Is%20Not%20Valid
      (Ability|EndAbility))))

(fn GetCustomAbilitySourceTransform
  (return (Utilities|Operators|Add (Transformation|GetActorLocation (CallFunction|GetAvatarasFortPlayerPawn)) (Math|Vector|MakeVector 0.0 0.0 (Components|Capsule|GetScaledCapsuleHalfHeight (Variables|/Script/Engine.Character|GetCapsuleComponent (CallFunction|GetAvatarasFortPlayerPawn))))) (Transformation|GetActorRotation (CallFunction|GetAvatarasFortPlayerPawn)) "1.000000,1.000000,1.000000"))
