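; Gameplay ability that makes a SouthBulk AI pawn play a montage, a cheer going by the asset name.
; On activation it casts the avatar to FortPawn and stores it as the AI pawn.
; It breaks the MontageInfo_MontageToPlay struct to feed a PlayMontageAndWaitNotify task.
; The ability ends once the montage completes, is interrupted or is cancelled, or immediately if the cast fails.
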(event Ability|EventActivateAbility
  (bind (_montageinfo_montagetoplay _montageinfo_animplayrate _montageinfo_animrootmotiontranslationscale _montageinfo_montagesectiontoplay _montageinfo_overridesection) (Utilities|Struct|BreakFortGameplayAbilityMontageInfo (Variables|GetMontageInfo_MontageToPlay)))
  (bind _asfort_pawn (Utilities|Casting|CastToFortPawn (Ability|GetAvatarActorfromActorInfo))
    (:then
      (Variables|InternalData|SetAIPawn _asfort_pawn)
      (Ability|Tasks|PlayMontageAndWaitNotify "None" _montageinfo_montagetoplay _montageinfo_animplayrate _montageinfo_overridesection true _montageinfo_animrootmotiontranslationscale
        (:OnCompleted
          _)
        (:OnInterrupted
          _)
        (:OnCancelled
          _))
      (Ability|EndAbility))
    (:CastFailed
      (Ability|EndAbility))))
