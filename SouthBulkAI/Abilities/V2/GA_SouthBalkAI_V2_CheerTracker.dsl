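; Gameplay ability for the SouthBulk AI (V2) that watches for its victims dying or being downed, to trigger a cheer.
; On activation it caches the avatar as a FortPawn and binds to its OnDidDamage event, unbinding when the ability ends.
; After each damage event it waits briefly, then applies the TriggerCheer gameplay effect to the owner if the damaged pawn has no health or is DBNO.
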(event Ability|EventActivateAbility
  (bind _asfort_pawn (Utilities|Casting|CastToFortPawn (Ability|GetAvatarActorfromActorInfo))
    (:then
      (bind _output_get (Variables|InternalData|SetAIPawn _asfort_pawn))
      (EventDispatchers|BindEventtoOnDidDamage _output_get (EventDispatchers|CreateEvent :!event Custom|AIDealtDamage)))
    (:CastFailed
      (Ability|EndAbility))))

(event Ability|EventOnEndAbility
  (params (bWasCancelled :type Boolean))
  (bind _aipawn (Variables|InternalData|GetAIPawn))
  (Utilities|IsValid _aipawn
    (:Is%20Valid
      (EventDispatchers|UnbindEventfromOnDidDamage _aipawn (EventDispatchers|CreateEvent :!event Custom|AIDealtDamage)))))

(event Custom|AIDealtDamage
  (params (DamagedActor :type Actor%20Object%20Reference) (Damage :type Float%20%28single-precision%29) (InstigatedBy :type Controller%20Object%20Reference) (DamageCauser :type Actor%20Object%20Reference) (HitLocation :type Vector) (FHitComponent :type Primitive%20Component%20Object%20Reference) (BoneName :type Name) (Momentum :type Vector))
  (bind _asfort_pawn (Utilities|Casting|CastToFortPawn DamagedActor))
  (Utilities|FlowControl|Delay 0.2)
  (if (or (<= (FortPawn|GetHealth _asfort_pawn) 0) (Health|IsDBNO _asfort_pawn))
    (Ability|ApplyGameplayEffectToOwner "/SouthBulkAI/Abilities/V2/GE_SouthBalkAI_TriggerCheer.GE_SouthBalkAI_TriggerCheer_C")))
