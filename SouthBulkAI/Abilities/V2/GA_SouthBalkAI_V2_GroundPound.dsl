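; South Bulk AI V2 ground pound ability: the AI performs a random number of pounds that spawn ground waves.
; Activation traces the floor once for the wave location, plays a speech bubble and rolls the count.
; DoPound plays a montage whose notify spawns a self-ignoring wave, fires a cue and re-calls DoPound after a delay.
; It requires walking out of water, commits after the final montage and applies a breather effect on end.
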(event Ability|EventActivateAbility
  (bind (_returnvalue_x _returnvalue_y _returnvalue_z) (Math|Vector|BreakVector (Transformation|GetActorLocation (Variables|InternalData|GetAIPawn))))
  (|CallParentFunction)
  (bind _asbp_player_pawn_non_participant_south_bulk_ai_v2 (Utilities|Casting|CastToBP_PlayerPawn_NonParticipant_SouthBulkAI_V2 (Variables|InternalData|GetAIPawn))
    (:then
      (Variables|InternalData|SetSouthBalkPawn _asbp_player_pawn_non_participant_south_bulk_ai_v2)
      (bind _outhit (Collision|LineTraceByChannel (Math|Vector|MakeVector _returnvalue_x _returnvalue_y (+ _returnvalue_z 2e+02)) (Math|Vector|MakeVector _returnvalue_x _returnvalue_y (- _returnvalue_z 2e+02)) "TraceTypeQuery1" false (Utilities|Array|MakeArray (Variables|InternalData|GetAIPawn)) "None" true "(R=1.000000,G=0.000000,B=0.000000,A=1.000000)" "(R=0.000000,G=1.000000,B=0.000000,A=1.000000)" 5.0))
      (Variables|Default|SetWaveSpawnLocation (+ (.ImpactPoint (Collision|BreakHitResult _outhit)) "0.000000,0.000000,20.000000"))
      (Class|BPFLAICommonFunctions|PlaySpeechBubble (Variables|InternalData|GetAIPawn) (Utilities|Array|RandomArrayItem (Variables|Default|GetTag_SpeechGroundPound)) false true)
      (Variables|InternalData|SetPoundCount (Math|Random|RandomIntegerinRange (Math|Float|Truncate (ScalableFloat|GetValueatLevel (Variables|Default|GetMinPoundCount))) (Math|Float|Truncate (ScalableFloat|GetValueatLevel (Variables|Default|GetMaxPoundCount)))))
      (Variables|InternalData|SetCurrentPoundCount 0)
      (CallFunction|DoPound))
    (:CastFailed
      (Ability|EndAbility))))

(event Ability|EventOnEndAbility
  (params (bWasCancelled :type Boolean))
  (|CallParentFunction bWasCancelled)
  (Ability|ApplyGameplayEffectToOwner (Variables|Default|GetGE_PlayerBreather)))

(event Custom|DoPound
  (bind (_montageinfo_montagetoplay _montageinfo_animplayrate _montageinfo_animrootmotiontranslationscale _montageinfo_montagesectiontoplay _montageinfo_overridesection) (Utilities|Struct|BreakFortGameplayAbilityMontageInfo (Variables|GetMontageInfo_MontageToPlay)))
  (bind _aipawn (Variables|InternalData|GetAIPawn))
  ; Do Pound
  (Ability|Tasks|PlayMontageAndWaitNotify "None" _montageinfo_montagetoplay _montageinfo_animplayrate _montageinfo_overridesection true _montageinfo_animrootmotiontranslationscale
    (:OnCompleted
      (if (not (< (Variables|InternalData|GetCurrentPoundCount) (Variables|InternalData|GetPoundCount)))
        (Ability|CommitAbility)
        (Ability|EndAbility)))
    (:OnNotifyBegin
      (bind _returnvalue (Game|SpawnActorfromClass "/SouthBulkAI/Gameplay/BGA_GroundWave_V2.BGA_GroundWave_V2_C" (Math|Transform|MakeTransform (Variables|Default|GetWaveSpawnLocation)) "AlwaysSpawn" "MultiplyWithRoot" _aipawn 0 _aipawn))
      (Variables|/SouthBulkAI/Gameplay/BGA_GroundWave.BGA_GroundWave_C|SetWaveIgnoreActors (Utilities|Array|MakeArray _aipawn) _returnvalue)
      (Ability|ExecuteGameplayCueOnOwner (Variables|Default|GetGC_GroundPound))
      (bind _result (Math|Integer|IncrementInt (Variables|InternalData|GetCurrentPoundCount))
        (:%20%20
          (if (< _result (Variables|InternalData|GetPoundCount))
            (Utilities|FlowControl|Delay (ScalableFloat|GetValueatLevel (Variables|Default|GetDelayBetweenPounds)))
            (CallFunction|DoPound)))))))

(fn CanActivateAbility
  (params (ActorInfo :type Gameplay%20Ability%20Actor%20Info%20Structure) (Handle :type Gameplay%20Ability%20Spec%20Handle%20Structure))
  (bind _asfort_player_pawn (Utilities|Casting|CastToFortPlayerPawn (Ability|GetAvatarActorfromActorInfo))
    (:then
      (switch Utilities|FlowControl|Switch|SwitchonEMovementMode (Variables|/Script/Engine.CharacterMovementComponent|GetMovementMode (Variables|/Script/Engine.Character|GetCharacterMovement _asfort_player_pawn))
        (:MOVE_None
          _)
        (:MOVE_Walking
          (Utilities|IsValid (WaterBody|GetCurrentWaterBody _asfort_player_pawn) :!label _label
            (:Is%20Valid
              (return 0 (not (> (WaterBody|GetWaterImmersionDepth _asfort_player_pawn) 0))))
            (:Is%20Not%20Valid
              (return 0 true))))
        (:MOVE_NavWalking
          (goto _label))
        (:MOVE_Falling
          _)
        (:MOVE_Swimming
          _)
        (:MOVE_Flying
          _)))
    (:CastFailed
      _))
  (return 0 false))
