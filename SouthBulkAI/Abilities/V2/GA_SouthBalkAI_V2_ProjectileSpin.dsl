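; AI gameplay ability for SouthBalk V2 that fires projectiles in a randomly picked horizontal, vertical, cross or straight-line pattern.
; Activation traces ahead for indestructible actors, then plays the pattern montage with a speech bubble.
; A looping timer runs waves that fire on montage notifies, while straight-line replaces it with delayed recursive shots.
; Activation requires a walking or swimming avatar; ending applies a breather effect and clears the timer.
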(event Custom|ValidateIndestructibleActor
  ; Make sure we don't have an indestructible collision infront of us before trying to throw projectiles.
  (bind (_outhit _returnvalue) (Collision|LineTraceByChannel (Transformation|GetActorLocation (Variables|InternalData|GetAIPawn)) (Utilities|Operators|Add (Transformation|GetActorLocation (Variables|InternalData|GetAIPawn)) (* (Transformation|GetActorForwardVector (Variables|InternalData|GetAIPawn)) 2e+02) (Math|Vector|MakeVector 0.0 0.0 (ScalableFloat|GetValueatLevel (Variables|Default|GetProjectileZOffset)))) "TraceTypeQuery1" false (Utilities|Array|MakeArray (Variables|InternalData|GetAIPawn))))
  (if _returnvalue
    (if (not (Variables|/Script/Engine.Actor|GetbCanBeDamaged (.HitActor (Collision|BreakHitResult _outhit))))
      (Ability|CommitAbility)
      (Ability|EndAbility) ; Don't use projectiles
    )))

(event Ability|EventActivateAbility
  (|CallParentFunction)
  (Variables|InternalData|SetCurrentProjectileCountStraightLine 0)
  (CallFunction|ValidateIndestructibleActor)
  (CallFunction|SelectPattern))

(event Ability|EventOnEndAbility
  (params (bWasCancelled :type Boolean))
  (|CallParentFunction bWasCancelled)
  (Ability|ApplyGameplayEffectToOwner (Variables|Default|GetGE_PlayerBreather))
  (Utilities|Time|ClearandInvalidateTimerbyHandle (Variables|InternalData|GetHDNL_ProjectileWaveTimer)))

(event Custom|FireHorizontalProjectiles
  ; Horizontal Projectiles
  (Variables|InternalData|SetProjectileOffsetCount 0)
  (for _index (range (+ (Math|Float|Truncate (- (ScalableFloat|GetValueatLevel (Variables|Default|GetProjectileCount)) 1.0)) 1))
    (CallFunction|FireProjectile (+ (* (Transformation|GetActorRightVector (Variables|InternalData|GetAIPawn)) (* (* (Variables|InternalData|GetProjectileOffsetCount) (ScalableFloat|GetValueatLevel (Variables|Default|GetProjectileOffset))) (Utilities|Select :Index (Math|Integer|%25%28Integer%29 _index 2) :Option%200 1 :Option%201 -1))) (Utilities|Operators|Add (* (Transformation|GetActorForwardVector (Variables|InternalData|GetAIPawn)) 1.5e+02) (Transformation|GetActorLocation (Variables|InternalData|GetAIPawn)) (Math|Vector|MakeVector 0.0 0.0 (ScalableFloat|GetValueatLevel (Variables|Default|GetProjectileZOffset))))) _index)
    (Variables|InternalData|SetProjectileOffsetCount (+ (Variables|InternalData|GetProjectileOffsetCount) (Utilities|Select :Index (Math|Integer|%25%28Integer%29 _index 2) :Option%200 1 :Option%201 0)))))

(event Custom|LaunchProjectiles
  ; Used when triggered from Spinning Animation
  (Projectile|SpawnProjectile :ProjectileClass "/SouthBulkAI/Gameplay/Prj_SouthBalkAI_V2.Prj_SouthBalkAI_V2_C" :RequestedBy (Variables|InternalData|GetAIPawn) :SpawnLocation (Utilities|Operators|Add (* (Transformation|GetActorForwardVector (Variables|InternalData|GetAIPawn)) 1.5e+02) (Transformation|GetActorLocation (Variables|InternalData|GetAIPawn)) (Math|Vector|MakeVector 0.0 0.0 (ScalableFloat|GetValueatLevel (Variables|Default|GetProjectileZOffset)))) :SpawnRotation "0, 0, 0" :SpawnDirection (Transformation|GetActorRotation (Variables|InternalData|GetAIPawn)) :EffectContainerSpecToApplyOnHit (Ability|MakeGameplayEffectContainerSpec "(TagName=\"Targeting.ApplicationTag.0\")") :bUseDefaultPhysics false :InitialSpeed (ScalableFloat|GetValueatLevel (Variables|Default|GetProjectileSpeed)) :GravityScale 0.0 :ChargePercent 0.0))

(event Custom|FireVerticalProjectiles
  ; Vertical Projectiles
  (for _index (range (+ (Math|Float|Truncate (- (ScalableFloat|GetValueatLevel (Variables|Default|GetProjectileCount)) 1.0)) 1))
    (CallFunction|FireProjectile (+ (Utilities|Operators|Add (* (Transformation|GetActorForwardVector (Variables|InternalData|GetAIPawn)) 1.5e+02) (Transformation|GetActorLocation (Variables|InternalData|GetAIPawn)) (Math|Vector|MakeVector 0.0 0.0 (* _index (ScalableFloat|GetValueatLevel (Variables|Default|GetProjectileOffsetVertical))))) (* (Transformation|GetActorRightVector (Variables|InternalData|GetAIPawn)) (* (* (ScalableFloat|GetValueatLevel (Variables|Default|GetProjectileOffset)) (Utilities|Select :Index (Math|Integer|%25%28Integer%29 (- (Variables|InternalData|GetCurrentProjectileWave) 1) 3) :Option%200 0.0 :Option%201 1.0 :Option%202 -1.0)) 1.5))) _index)))

(event Custom|SelectPattern
  ; Select which Pattern and start to fire it
  (bind _targetactor (Class|BPFLAICommonFunctions|GetTargetActorfromBlackboard (Variables|InternalData|GetAIPawn)))
  (Variables|InternalData|SetTargetFortPawn _targetactor)
  (Variables|InternalData|SetSelectedPattern (Utilities|Select :Index (Math|Random|RandomIntegerinRange 0 3) :Option%200 "NewEnumerator0" :Option%201 "NewEnumerator1" :Option%202 "NewEnumerator2" :Option%203 "NewEnumerator4" :Option%204 "NewEnumerator3" :Option%205 "NewEnumerator5"))
  (Variables|InternalData|SetCurrentProjectileWave 0)
  (Ability|Tasks|PlayMontageAndWaitNotify "None" (Utilities|Select :Index (Variables|InternalData|GetSelectedPattern) :NewEnumerator0 (Variables|Default|GetMontage_Horizontal) :NewEnumerator1 (Variables|Default|GetMontage_Vertical) :NewEnumerator2 (Variables|Default|GetMontage_Cross) :NewEnumerator3 (Variables|Default|GetMontage_Cross) :NewEnumerator4 (Variables|Default|GetMontage_Cross) :NewEnumerator5 (Variables|Default|GetMontage_Cross))
    (:then
      (Class|BPFLAICommonFunctions|PlaySpeechBubble (Variables|InternalData|GetAIPawn) (Utilities|Array|RandomArrayItem (Variables|Default|GetTag_SpeechProjectile)) false true))
    (:OnCompleted
      _)
    (:OnInterrupted
      _)
    (:OnCancelled
      _))
  (bind _returnvalue (Utilities|Time|SetTimerbyEvent (EventDispatchers|CreateEvent :!event Custom|StartProjectileWave) (ScalableFloat|GetValueatLevel (Variables|Default|GetProjectileWaveDelay)) true))
  (Variables|InternalData|SetHDNL_ProjectileWaveTimer _returnvalue)
  (CallFunction|StartProjectileWave))

(event Custom|StartProjectileWave
  ; Loop that fires Projectile Waves
  (if (== (Variables|InternalData|GetCurrentProjectileWave) (ScalableFloat|GetValueatLevel (Variables|Default|GetProjectileWaveCount)))
    (Utilities|Time|ClearandInvalidateTimerbyHandle (Variables|InternalData|GetHDNL_ProjectileWaveTimer))
    (CallFunction|ProjectilesDone)
    (else
      (Math|Integer|IncrementInt (Variables|InternalData|GetCurrentProjectileWave)
        (:%20%20
          (Ability|Tasks|PlayMontageAndWaitNotify "None" (Variables|Default|GetMontage_Fire)
            (:OnNotifyBegin
              (switch Utilities|FlowControl|Switch|SwitchonENUM_SouthBalkAI_ProjectilePattern (Variables|InternalData|GetSelectedPattern)
                (:NewEnumerator0
                  (CallFunction|FireHorizontalProjectiles))
                (:NewEnumerator1
                  (CallFunction|FireVerticalProjectiles))
                (:NewEnumerator2
                  (CallFunction|FireCrossProjectiles))
                (:NewEnumerator4
                  (Utilities|Time|ClearandInvalidateTimerbyHandle (Variables|InternalData|GetHDNL_ProjectileWaveTimer))
                  (CallFunction|FireStraightLineProjectiles))))))))))

(event Custom|FireCrossProjectiles
  (bind _projectileoffsetcross (Variables|Default|GetProjectileOffsetCross))
  (bind _returnvalue (ScalableFloat|GetValueatLevel _projectileoffsetcross))
  (bind _projectileoffsetcount (Variables|InternalData|GetProjectileOffsetCount))
  (bind _returnvalue_2 (* _returnvalue _projectileoffsetcount))
  (bind _returnvalue_3 (* _returnvalue_2 -1.0))
  ; Cross Projectiles
  (Variables|InternalData|SetProjectileOffsetCount 1)
  (for _index (range (+ (Math|Float|Truncate (- (ScalableFloat|GetValueatLevel (Variables|Default|GetProjectileCount)) 1.0)) 1))
    (if (== _index 0)
      (Variables|InternalData|SetVerticalOffset 0.0)
      (Variables|InternalData|SetHorizontalOffset 0.0)
      (else
        (switch int (Math|Integer|%25%28Integer%29 _index 5)
          (:0
            (Math|Integer|IncrementInt (Variables|InternalData|GetProjectileOffsetCount)))
          (:1
            (Variables|InternalData|SetVerticalOffset _returnvalue_2)
            (Variables|InternalData|SetHorizontalOffset _returnvalue_2))
          (:2
            (Variables|InternalData|SetVerticalOffset _returnvalue_2)
            (Variables|InternalData|SetHorizontalOffset _returnvalue_3))
          (:3
            (Variables|InternalData|SetVerticalOffset _returnvalue_3)
            (Variables|InternalData|SetHorizontalOffset _returnvalue_2))
          (:4
            (Variables|InternalData|SetVerticalOffset _returnvalue_3)
            (Variables|InternalData|SetHorizontalOffset _returnvalue_3)))))
    (CallFunction|FireProjectile (Utilities|Operators|Add (* (Transformation|GetActorForwardVector (Variables|InternalData|GetAIPawn)) 1.5e+02) (Transformation|GetActorLocation (Variables|InternalData|GetAIPawn)) (* (Transformation|GetActorRightVector (Variables|InternalData|GetAIPawn)) (Variables|InternalData|GetVerticalOffset)) (Math|Vector|MakeVector 0.0 0.0 (+ (ScalableFloat|GetValueatLevel (Variables|Default|GetProjectileZOffset)) (Variables|InternalData|GetHorizontalOffset)))) _index)))

(event Custom|ProjectilesDone
  (Ability|CommitAbility)
  (Ability|EndAbility))

(event Custom|FireStraightLineProjectiles
  ; Fire in a sraight line
  (CallFunction|FireProjectile (Utilities|Operators|Add (* (Transformation|GetActorForwardVector (Variables|InternalData|GetAIPawn)) 1.5e+02) (Transformation|GetActorLocation (Variables|InternalData|GetAIPawn)) (Math|Vector|MakeVector 0.0 0.0 (ScalableFloat|GetValueatLevel (Variables|Default|GetProjectileZOffset)))))
  (Math|Integer|IncrementInt (Variables|InternalData|GetCurrentProjectileCountStraightLine)
    (:%20%20
      (if (> (Variables|InternalData|GetCurrentProjectileCountStraightLine) (ScalableFloat|GetValueatLevel (Variables|Default|GetProjectileStraightMaxCount)))
        (CallFunction|ProjectilesDone)
        (else
          (Utilities|FlowControl|Delay (ScalableFloat|GetValueatLevel (Variables|Default|GetProjectileStraightDelay)))
          (CallFunction|FireStraightLineProjectiles))))))

(event Custom|FireProjectile
  (params (SpawnLocation :type Vector) (ProjectileIndex :type Integer)))

(fn AdjustSpawnLocationFromPitch
  (params (InitialStartLocation :type Vector))
  (if (> (Class|AbilityFunctions|GetAngleBetweenTwoVectors (Transformation|GetActorForwardVector (Variables|InternalData|GetAIPawn)) (Math|Vector|Normalize (- (Transformation|GetActorLocation (Variables|InternalData|GetTargetFortPawn)) (Transformation|GetActorLocation (Variables|InternalData|GetAIPawn))))) (ScalableFloat|GetValueatLevel (Variables|Default|GetAngleThreshold)))
    (LocalVariables|SetNewStartLocation (Utilities|Operators|Add InitialStartLocation (Math|Vector|MakeVector 0.0 0.0 (ScalableFloat|GetValueatLevel (Variables|Default|GetIncreasedZFromAngleThreshold)))))
    (else
      (LocalVariables|SetNewStartLocation InitialStartLocation)))
  (return (LocalVariables|GetNewStartLocation)))

(fn CanActivateAbility
  (params (ActorInfo :type Gameplay%20Ability%20Actor%20Info%20Structure) (Handle :type Gameplay%20Ability%20Spec%20Handle%20Structure))
  (bind _asfort_player_pawn (Utilities|Casting|CastToFortPlayerPawn (Ability|GetAvatarActorfromActorInfo))
    (:then
      (switch Utilities|FlowControl|Switch|SwitchonEMovementMode (Variables|/Script/Engine.CharacterMovementComponent|GetMovementMode (Variables|/Script/Engine.Character|GetCharacterMovement _asfort_player_pawn))
        (:MOVE_None
          _)
        (:MOVE_Walking
          (return 0 true))
        (:MOVE_NavWalking
          (return 0 true))
        (:MOVE_Falling
          _)
        (:MOVE_Swimming
          (return 0 true))
        (:MOVE_Flying
          _)))
    (:CastFailed
      _))
  (return 0 false))

(fn SetOGTransforms
  (bind _aipawn (Variables|InternalData|GetAIPawn))
  (Variables|Default|SetOGRightVector (Transformation|GetActorRightVector _aipawn))
  (Variables|Default|SetOGForwardVector (Transformation|GetActorForwardVector _aipawn))
  (Variables|Default|SetOGActorLocation (Transformation|GetActorLocation _aipawn))
  (return))
